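; BP_FMCalibration_Helper is an actor that runs the metronome used for calibration.
; StartMetronome picks a calibrometer, vocal, hi-hat or plain sound, creates a MetaSound-driven music clock from its MIDI Clock output, then sets tempo and silent-aware gain.
; GetVideoFractionalBeat reads beat distance at video render time, flagging a first frame whenever that value wraps below the previous one.
; StopMetronome, also called on EndPlay, stops audio and clears the clock to reset the track.
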(event EventEndPlay
  (params (EndPlayReason :type EEndPlayReason%20Enum :default "Destroyed"))
  (CallFunction|StopMetronome))

(fn GetVideoFractionalBeat
  (bind _calibrationclock (Variables|Default|GetCalibrationClock))
  (bind _returnvalue (MusicClock|GetDistancefromCurrentBeat _calibrationclock "VideoRenderTime" :!class /Script/HarmonixMetasound.MusicClockComponent))
  (Utilities|IsValid _calibrationclock
    (:Is%20Valid
      (bind _output_get (LocalVariables|SetFirstFrame (< _returnvalue (Variables|Default|GetLastFractionalBeat))))
      (bind _output_get_2 (Variables|Default|SetLastFractionalBeat _returnvalue))
      (return _output_get_2 _output_get))
    (:Is%20Not%20Valid
      (return 0.0 true))))

(fn StartMetronome
  (params (Tempo :type Float%20%28double-precision%29) (Silent :type Boolean) (UseHiHats :type Boolean) (Use%20Calibrometer :type Boolean) (ProVocals :type Boolean))
  (Audio|Components|Audio|SetSound (Variables|Default|GetAudio) (select Use%20Calibrometer (select ProVocals "/FMCalibration/Sounds/CalibrometerVocals.CalibrometerVocals" "/FMCalibration/Sounds/CalibrometerMetronome.CalibrometerMetronome") (select UseHiHats "/FMCalibration/Sounds/CalibrationMetronomeWithBlips.CalibrationMetronomeWithBlips" "/FMCalibration/Sounds/CalibrationMetronome.CalibrationMetronome")))
  (bind _returnvalue (Audio|MusicClock|CreateMetasoundDrivenMusicClock (Variables|Default|GetAudio) "MIDI Clock" false))
  (Variables|Default|SetCalibrationClock _returnvalue)
  (Audio|Parameter|SetFloatParameter (Variables|Default|GetAudio) "Tempo" Tempo :!class /Script/AudioExtensions.AudioParameterControllerInterface)
  (Audio|Parameter|SetFloatParameter (Variables|Default|GetAudio) "Gain" (Math|Float|SelectFloat 0.0 1.0 (Variables|GetSilent)) :!class /Script/Engine.AudioComponent)
  (Components|Activation|Activate (Variables|Default|GetAudio) false)
  (MusicClock|Start (Variables|Default|GetCalibrationClock))
  (Variables|Default|SetLastFractionalBeat -1.0))

(fn StopMetronome
  (Audio|Components|Audio|Stop (Variables|Default|GetAudio))
  (Variables|Default|SetCalibrationClock 0) ; Clear to force the track back to zero.
)

(fn ConstructionScript)
